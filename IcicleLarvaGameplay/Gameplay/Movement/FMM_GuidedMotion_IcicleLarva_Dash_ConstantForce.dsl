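; Guided-motion movement extension for the IcicleLarva dash, driven by a constant force.
; The force direction comes from the parent, with Z flattened to zero while control pitch sits within the punch pitch deadzone.
; Activation applies a passive gameplay effect and deactivation removes effects granting the Dash status tag.
; That separate tag serves sim-proxy animation, avoiding client re-activation being blocked while the server removes the effect.
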(event EventOnExtensionActivate
  (params (RuntimeContext :type MMERuntime%20Context%20Structure%20%28by%20ref%29))
  ; Adding separate tag for sim proxies (anim). In future versions (37.00) TagsToPush should be replicated to sim proxies anyway as sim proxies will receive the activate/deactivate calls, so that could be used instead.
  ;
  ; The issue with using the same guided motion MME tag would be that the local client will have to wait for the server to remove the active effect, so activating the MME will be blocked at the end of a dash because the local client thinks the guided motion MME is still active.
  (GameplayEffects|ApplyGameplayEffectToSelf (Ability|GetAbilitySystemComponent (Variables|FortMovementMode_BaseExtLogic|GetOwnerPawn)) (Variables|Default|GetPassiveGEClass) 0.0 (Class|AthenaFunctionLibrary|MakeSimpleFortEffectContext (Variables|FortMovementMode_BaseExtLogic|GetOwnerPawn) 0 self)))

(event EventOnExtensionDeactivate
  (params (RuntimeContext :type MMERuntime%20Context%20Structure%20%28by%20ref%29))
  ; Adding separate tag for sim proxies (anim). In future versions (37.00) TagsToPush should be replicated to sim proxies anyway as sim proxies will receive the activate/deactivate calls, so that could be used instead.
  ;
  ; The issue with using the same guided motion MME tag would be that the local client will have to wait for the server to remove the active effect, so activating the MME will be blocked at the end of a dash because the local client thinks the guided motion MME is still active.
  (GameplayEffects|RemoveActiveEffectswithGrantedTags (Ability|GetAbilitySystemComponent (Variables|FortMovementMode_BaseExtLogic|GetOwnerPawn)) "(GameplayTags=((TagName=\"Gameplay.Status.IcicleLarva.Dash\")))"))

(fn GetDirection_ConstantForce
  (params (GuidedMotionRuntimeData :type Fort%20Movement%20Mode%20Guided%20Motion%20Runtime%20Data%20Object%20Reference))
  (bind (_returnvalue_roll _returnvalue_pitch) (Math|Rotator|BreakRotator (Pawn|GetControlRotation (Variables|FortMovementMode_BaseExtLogic|GetOwnerPawn) :!class /Script/Engine.Pawn)))
  (bind _returnvalue (|CallParentFunction GuidedMotionRuntimeData))
  (return (Math|Vector|MakeVector (.x _returnvalue) (.y _returnvalue) (Math|Float|SelectFloat 0.0 (.z _returnvalue) (<= (Math|Float|Absolute(Float) (Math|Rotator|NormalizeAxis _returnvalue_pitch)) (ScalableFloat|GetValueatLevel (Variables|Default|GetPunchPitchDeadzone)))))))
